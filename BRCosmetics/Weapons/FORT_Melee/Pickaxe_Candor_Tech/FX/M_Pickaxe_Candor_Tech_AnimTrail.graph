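# Translucent, unlit, two-sided material for the Candor Tech pickaxe animation trail, usable on Niagara ribbons and mesh particles.
# Color comes from a panning texture, desaturated by a dynamic parameter, tinted by particle color and passed through an emissive-adjustment function.
# The alpha chain pans a square trail texture with power-warped U and optional rotation, then multiplies in particle alpha, U and sine-based V edge masks, scanlines and near-camera fade.
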Material /BRCosmetics/Weapons/FORT_Melee/Pickaxe_Candor_Tech/FX/M_Pickaxe_Candor_Tech_AnimTrail
graph {
 properties {
  bUsedWithNiagaraMeshParticles=true
  bUsedWithNiagaraRibbons=true
  blendMode=BLEND_Translucent
  pixelDepthOffsetMode=PDOM_Legacy
  shadingModel=MSM_Unlit
  twoSided=true
  usedShadingModels=Unlit
 }
 nodes [
  n1 = TextureCoordinate()
  n2 = ComponentMask(None=VectorParameter() {parameterName=Color Tex UV, defaultValue=<rgba 2 1 0 0>}.RGB) {g=true, r=true}
  n3 = TextureSampleParameter2D(UVs=Panner(Coordinate=Multiply(A=@n1, B=@n2)) {speedX=-1}) {parameterName=Color Tex, samplerType=SAMPLERTYPE_LinearColor, texture=/Game/Effects/Fort_Effects/BluGlow/BluGlowAnimation_Normals}
  n4 = DynamicParameter() {defaultValue=<rgba 0 1 1 1>, paramNames [des,Param2,Param3,Param4]}
  n5 = ParticleColor()
  MaterialFunctionCall(Emissive=Multiply(A=Desaturation(None=@n3.RGB, Fraction=@n4.des), B=@n5.RGB)) {materialFunction=&MF_FX_EmissiveAdjustments}
  n6 = ScalarParameter() {parameterName=Near Camera Fade}
  n7 = Multiply(A=TextureCoordinate(), B=ComponentMask(None=VectorParameter() {parameterName=UV_Scale, defaultValue=<rgba 2 1.5 0 0>}.RGB) {g=true, r=true})
  n8 = ComponentMask(None=@n7) {r=true}
  n9 = ScalarParameter() {defaultValue=2, parameterName=U Scale Power}
  n10 = Panner(
   Coordinate=AppendVector(A=Power(Base=@n8, Exp=@n9), B=ComponentMask(None=@n7) {g=true})
   Time=LinearInterpolate(A=Constant() {r=1}, B=Time(), Alpha=ScalarParameter() {parameterName=Use loop time panner})
   Speed=ComponentMask(None=VectorParameter() {parameterName=Offset, defaultValue=<rgba 0 0.4 0 0>}.RGB) {g=true, r=true}
  ) {speedY=0.4}
  n11 = Constant() {r=1}
  n12 = TextureSampleParameter2D(UVs=LinearInterpolate(A=@n10, B=Rotator(Coordinate=@n10, Time=@n11) {speed=0.85}, Alpha=ScalarParameter() {defaultValue=1, parameterName=Use Rotator})) {
   parameterName=Default Tex
   samplerType=SAMPLERTYPE_LinearGrayscale
   texture=&/T_AnimTrail_Square01
  }
  n13 = TextureCoordinate()
  n14 = OneMinus(None=@n13)
  n15 = Multiply(A=ComponentMask(None=@n13) {r=true}, B=Power(Base=ComponentMask(None=@n14) {r=true}))
  n16 = Multiply(A=Multiply(A=@n5.A, B=Power(Base=@n12.R)), B=Saturate(None=Multiply(A=@n15) {constB=5}))
  n17 = Sine(None=ComponentMask(None=TextureCoordinate()) {g=true}) {period=2}
  n18 = ScalarParameter() {defaultValue=7, parameterName=Mask Power}
  n19 = Saturate(None=Multiply(A=Power(Base=@n17, Exp=@n18) {constExponent=7}) {constB=10})
  n20 = ScalarParameter() {defaultValue=50, parameterName=Scanline Amount}
  n21 = ScalarParameter() {defaultValue=-0.15, parameterName=ScanlineSize}
  n22 = ScalarParameter() {defaultValue=0.5, parameterName=Scanlines Dark}
  n23 = ScalarParameter() {defaultValue=1, parameterName=Scanlines Light}
  Multiply(
   A=Power(
    Base=MaterialFunctionCall(Distance=@n6) {materialFunction=&Near_Camera_Fade}.Result
    Exp=ScalarParameter() {defaultValue=2, parameterName=Camera Fade Pow}
   )
   B=Multiply(
    A=Multiply(A=@n16, B=@n19)
    B=MaterialFunctionCall!(
     Number of Scanlines=@n20
     Scanline Speed
     "Scanlines Thickness (Bias)"=@n21
     Scanlines Dark Value=@n22
     Scanlines Light Value=@n23
     Block Opacity
     UVs
     Time
     Block Repeat
    ) {materialFunction=/Game/Effects/Fort_Effects/Effects/Atlas/Materials/MF_Atlas_Scanlines_01}.Result
   )
  )
  TextureCoordinate() {uTiling=2, vTiling=1.5}
  TextureCoordinate() {uTiling=2}
 ]
}
