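# Physics asset for the Candor Tech pickaxe mesh, defining collision bodies and joints for its dangling strap and wires.
# Two capsules on the two-handed handle form the base that three chains attach to: a short right strap and five-segment left and right wires.
# Small spheres and capsules make up the chain links, joined by swing and twist limits, with the wire roots swinging less than the links below them.
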PhysicsAsset /BRCosmetics/Weapons/FORT_Melee/Pickaxe_Candor_Tech/Meshes/Candor_Tech_Axe_Physics
bodies {
 Dyn_strap_R_01 [{shapeType=Sphere, radius=1, shapeName=Dyn_strap_R_01_sphere, center=<0 -0.2675 0>}]
 Dyn_strap_R_02 [{shapeType=Capsule, length=1.5, radius=1.25, shapeName=Dyn_strap_R_02_capsule, center=<0 -1.1382 0>}]
 Dyn_wire_L_01 [{shapeType=Sphere, radius=0.7, shapeName=Dyn_wire_L_01_sphere}]
 Dyn_wire_L_02 [{shapeType=Capsule, length=1.0057, radius=1, shapeName=Dyn_wire_L_02_capsule, rotation=<rot 0 -90 0>}]
 Dyn_wire_L_03 [{shapeType=Capsule, length=1.0091, radius=0.8, shapeName=Dyn_wire_L_03_capsule}]
 Dyn_wire_L_04 [{shapeType=Capsule, length=1.2404, radius=0.78, shapeName=Dyn_wire_L_04_capsule, rotation=<rot 0 -90 0>}]
 Dyn_wire_L_05 [{shapeType=Capsule, length=1.1402, radius=0.8598, shapeName=Dyn_wire_L_05_capsule, rotation=<rot 0 -90 0>}]
 Dyn_wire_R_01 [{shapeType=Sphere, radius=0.7, shapeName=Dyn_wire_R_01_sphere}]
 Dyn_wire_R_02 [{shapeType=Capsule, length=0.7853, radius=1, shapeName=Dyn_wire_R_02_capsule, rotation=<rot 0 -90 0>}]
 Dyn_wire_R_03 [{shapeType=Capsule, length=1, radius=0.8, shapeName=Dyn_wire_R_03_capsule}]
 Dyn_wire_R_04 [{shapeType=Capsule, length=1.2101, radius=0.7899, shapeName=Dyn_wire_R_04_capsule, rotation=<rot 0 -90 0>}]
 Dyn_wire_R_05 [{shapeType=Capsule, length=1.1402, radius=0.8598, shapeName=Dyn_wire_R_05_capsule, rotation=<rot 0 -90 0>}]
 Melee_TwoHanded_Handle [
  {shapeType=Capsule, length=38.5, radius=2, shapeName=Melee_TwoHanded_Handle_capsule, center=<0 0 1.9007>}
  {shapeType=Capsule, length=18.5, radius=5.25, shapeName=Melee_TwoHanded_Handle_capsule, center=<0 0 33.2989>}
 ]
}
constraints [
 {bone1Name=Dyn_strap_R_02, bone2Name=Dyn_strap_R_01, swing1LimitDegrees=60, swing2LimitDegrees=10, twistLimitDegrees=10}
 {bone1Name=Dyn_strap_R_01, bone2Name=Melee_TwoHanded_Handle, swing1LimitDegrees=60, swing2LimitDegrees=15, twistLimitDegrees=15}
 {bone1Name=Dyn_wire_R_05, bone2Name=Dyn_wire_R_04, swing1LimitDegrees=45, swing2LimitDegrees=45, twistLimitDegrees=40}
 {bone1Name=Dyn_wire_R_04, bone2Name=Dyn_wire_R_03, swing1LimitDegrees=45, swing2LimitDegrees=45, twistLimitDegrees=40}
 {bone1Name=Dyn_wire_R_03, bone2Name=Dyn_wire_R_02, swing1LimitDegrees=45, swing2LimitDegrees=45, twistLimitDegrees=40}
 {bone1Name=Dyn_wire_R_02, bone2Name=Dyn_wire_R_01, swing1LimitDegrees=45, swing2LimitDegrees=45, twistLimitDegrees=40}
 {bone1Name=Dyn_wire_R_01, bone2Name=Melee_TwoHanded_Handle, swing1LimitDegrees=25, swing2LimitDegrees=25, twistLimitDegrees=40}
 {bone1Name=Dyn_wire_L_05, bone2Name=Dyn_wire_L_04, swing1LimitDegrees=45, swing2LimitDegrees=45, twistLimitDegrees=40}
 {bone1Name=Dyn_wire_L_04, bone2Name=Dyn_wire_L_03, swing1LimitDegrees=45, swing2LimitDegrees=45, twistLimitDegrees=40}
 {bone1Name=Dyn_wire_L_03, bone2Name=Dyn_wire_L_02, swing1LimitDegrees=45, swing2LimitDegrees=45, twistLimitDegrees=40}
 {bone1Name=Dyn_wire_L_02, bone2Name=Dyn_wire_L_01, swing1LimitDegrees=45, swing2LimitDegrees=45, twistLimitDegrees=40}
 {bone1Name=Dyn_wire_L_01, bone2Name=Melee_TwoHanded_Handle, swing1LimitDegrees=25, swing2LimitDegrees=25, twistLimitDegrees=40}
]
