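# Physics asset for the Candor_Tech_Pink_Axe_L pickaxe mesh, providing collision bodies and joints for its handle and dangling upper strap.
# A long capsule covers the two-handed handle bone, while a small sphere and a short capsule cover the two strap bones.
# Angular-limited constraints chain the strap segments to the handle, with a much wider range on the first swing axis than on the second swing or twist.
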PhysicsAsset /BRCosmetics/Weapons/FORT_Melee/Pickaxe_Candor_Tech/Meshes/Candor_Tech_Pink_Axe_L_Physics
bodies {
 Dyn_strap_upper_01 [{shapeType=Sphere, radius=1, shapeName=Dyn_strap_upper_01_sphere, center=<0.157 0 0>}]
 Dyn_strap_upper_02 [{shapeType=Capsule, length=1.75, radius=1.1, shapeName=Dyn_strap_upper_02_capsule, center=<0 -0.9522 0>}]
 Melee_TwoHanded_Handle [{shapeType=Capsule, length=44.9454, radius=2.185, shapeName=Melee_TwoHanded_Handle_capsule, center=<0 0 2.5129>}]
}
constraints [
 {bone1Name=Dyn_strap_upper_01, bone2Name=Melee_TwoHanded_Handle, swing1LimitDegrees=60, swing2LimitDegrees=12, twistLimitDegrees=10}
 {bone1Name=Dyn_strap_upper_02, bone2Name=Dyn_strap_upper_01, swing1LimitDegrees=60, swing2LimitDegrees=15, twistLimitDegrees=5}
]
